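; BGA_BigBall_FireMesh is an actor Blueprint for the BigBall NPC ability that configures fire ignition and spread through Curie fuel settings.
; After calling the parent construction script, it evaluates scalable float data and writes it to those settings.
; One truncated fuel value sets both self-propagation and nearby-material propagation, and one magnitude sets both self and nearby fire strength.
; Separate data values set ignition radii for nearby materials and landscape.
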(fn ConstructionScript
  (bind _data_ignitenearbymaterialsfiremagnitude (Variables|Default|GetData_IgniteNearbyMaterialsFireMagnitude))
  (bind _returnvalue (ScalableFloat|GetValueatLevel _data_ignitenearbymaterialsfiremagnitude))
  (bind _data_selfpropagationfuel (Variables|Default|GetData_SelfPropagationFuel))
  (bind _returnvalue_2 (ScalableFloat|GetValueatLevel _data_selfpropagationfuel))
  (bind _returnvalue_3 (Math|Float|Truncate _returnvalue_2))
  (|CallParentFunction)
  (Variables|CurieFuel|SetSelfPropagationFuel _returnvalue_3)
  (Variables|CurieFuel|SetIgniteNearbyMaterialsPropogationFuel _returnvalue_3)
  (Variables|CurieFuel|SetIgniteNearbyMaterialsRadius (ScalableFloat|GetValueatLevel (Variables|Default|GetData_IgniteNearbyMaterialsRadius)))
  (Variables|CurieFuel|SetIgniteNearbyMaterialsLandscapeIgnitionRadius (ScalableFloat|GetValueatLevel (Variables|Default|GetData_IgniteNearbyLandscapeRadius)))
  (Variables|CurieFuel|SetIgniteNearbyMaterialsFireMagnitude _returnvalue)
  (Variables|CurieFuel|SetSelfFireMagnitude _returnvalue))
