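; Gameplay ability for an NPC's BigBall attack, a hold-and-release projectile power built on a parent ability.
; On activation it measures distance to the blackboard goal target and sets spawn pitch offset from the first matching entry in parallel distance and pitch arrays.
; It then waits a scalable hold duration before calling Released.
; Activation keeps the parent's check only while walking, nav-walking or swimming, and fails otherwise.
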(event Ability|EventActivateAbility
  (bind _returnvalue (Ability|GetAvatarActorfromActorInfo))
  (|CallParentFunction)
  (bind _asfort_pawn (Utilities|Casting|CastToFortPawn (AI|Components|Blackboard|GetValueasObject (AI|GetBlackboard _returnvalue) (Variables|Default|GetGoalTargetKeyname)))
    (:then
      (bind _distance (Class|MissionBlueprintFunctionLibrary|DistanceBetweenTwoVectors (Transformation|GetActorLocation _asfort_pawn) (Transformation|GetActorLocation _returnvalue)))
      (CallFunction|UpdateProjectilePitchfromDistance _distance))
    (:CastFailed
      _))
  (bind _output_get (Variables|Default|SetHoldDuration (ScalableFloat|GetValueatLevel (Variables|Default|GetSF_HoldDuration))))
  (Utilities|FlowControl|Delay _output_get)
  (CallFunction|Released _output_get))

(event Ability|EventOnEndAbility
  (params (bWasCancelled :type Boolean))
  (|CallParentFunction bWasCancelled) ; This node is disabled and will not be called.\nDrag off pins to build functionality.
)

(fn CanActivateAbility
  (params (ActorInfo :type Gameplay%20Ability%20Actor%20Info%20Structure) (Handle :type Gameplay%20Ability%20Spec%20Handle%20Structure))
  (bind (_relevanttags _returnvalue) (|CallParentFunction ActorInfo Handle))
  (bind _asfort_pawn (Utilities|Casting|CastToFortPawn (Ability|GetAvatarActorfromActorInfo))
    (:then
      (switch Utilities|FlowControl|Switch|SwitchonEMovementMode (Variables|/Script/Engine.CharacterMovementComponent|GetMovementMode (Variables|/Script/Engine.Character|GetCharacterMovement _asfort_pawn))
        (:MOVE_None
          _)
        (:MOVE_Walking
          (return _relevanttags _returnvalue))
        (:MOVE_NavWalking
          (return _relevanttags _returnvalue))
        (:MOVE_Falling
          _)
        (:MOVE_Swimming
          (return _relevanttags _returnvalue))
        (:MOVE_Flying
          _)
        (:MOVE_Custom
          _)))
    (:CastFailed
      _))
  (return _relevanttags false))

(fn UpdateProjectilePitchFromDistance
  (params (Distance :type Float%20%28double-precision%29))
  (for (_array_element _array_index) (Variables|Default|GetDistancesToTarget)
    (if (<= (Variables|GetDistance) (ScalableFloat|GetValueatLevel _array_element))
      (Variables|PowerProjectileSpawn|SetRow_SpawnPitchOffset (Utilities|Array|Get(acopy) (Variables|Default|GetPitchAdjustmentsRelatedtoTarget) _array_index))
      (return)))
  (return))
