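; BigBall projectile for the NPC library that explodes on hitting something or entering water.
; The first server-side hit explodes it, sets a RepNotify that hides it, disables collision and fades audio, then shortens its lifespan.
; When Curie and a scalable float allow fire, the hit (unless already in water) and the explosion each spawn a fire actor.
; Water entry plays surface-aligned splash effects; clients disable capsule overlap events.
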(event EventBeginPlay
  (|CallParentFunction)
  (Variables|Default|SetFort%20Ability%20System%20Component (Ability|GetFortAbilitySystemComponent (Game|GetInstigator)))
  (Networking|CanExecuteCosmeticEvents
    (:True
      (Variables|Default|SetStarting%20Location (Transformation|GetActorLocation)))))

(event OnEnterWaterEvent(WaterInteractionComponent)
  (params (WaterBody :type Fort%20Water%20Body%20Actor%20Object%20Reference) (WaterInteractionComponent :type Fort%20Water%20Interaction%20Component%20Object%20Reference) (bIsFirstBody :type Boolean))
  (bind (_waterplanelocation _waterplanenormal) (FX|GetLastWaterSurfaceInfo WaterInteractionComponent :!class /Script/FortniteGame.FortWaterInteractionComponent))
  ;[1 1 1] Splash FX
  (CallFunction|SplashFX (Math|Rotator|MakeRotfromX _waterplanenormal))
  (Variables|Default|SetHitWater true)
  (Delivery|DoExplosion))

(event Custom|SplashFX
  (params (Rotation :type Rotator))
  (bind _returnvalue (Transformation|GetActorLocation))
  ;[1 1 1] Splash FX
  (Audio|PlaySoundatLocation (Variables|Default|GetExplodeWaterSound) _returnvalue)
  (Effects|SpawnEmitteratLocation "/NPCLibrary/Abilities/BigBall/FX/P_BigBall_Impact_Water.P_BigBall_Impact_Water" (- _returnvalue (Variables|Default|GetSplashFXZOffset)) Rotation))

(event Collision|EventHit
  (params (MyComp :type Primitive%20Component%20Object%20Reference) (Other :type Actor%20Object%20Reference) (OtherComp :type Primitive%20Component%20Object%20Reference) (bSelfMoved :type Boolean) (HitLocation :type Vector :default "0, 0, 0") (HitNormal :type Vector :default "0, 0, 0") (NormalImpulse :type Vector :default "0, 0, 0") (Hit :type Hit%20Result%20Structure%20%28by%20ref%29))
  ;[1 1 1] On Hit or enter water do the Explode, Also set the rep notify to hide the actor / turn off collision. Set the lifespan to kill the projectile
  (Utilities|FlowControl|DoOnce false
    (:Completed
      (if (Networking|IsServer)
        (if (Math|Boolean|ANDBoolean (Curie|IsCurieActive) (Math|Conversions|ToBoolean(Integer) (Math|Float|Truncate (ScalableFloat|GetValueatLevel (Variables|Default|GetRowFireEnabled)))) (not (Variables|Default|GetHitWater)))
          (Building|SpawnBuildingGameplayActor (Variables|Default|GetFireMesh) (Math|Transform|MakeTransform HitLocation) (Variables|Actor|GetInstigator))
          (else
            _))
        (Delivery|DoExplosion)
        (Variables|Default|SetHideAndKill true)
        (Actor|SetLifeSpan 0.2)))))

(event EventOnExploded
  (params (HitActors :type Array%20of%20Actor%20Object%20References) (HitResults :type Array%20of%20Hit%20Result%20Structures))
  (bind _returnvalue (Transformation|GetActorLocation))
  (bind _explodesound (Variables|Default|GetExplodeSound))
  ;[1 1 1] SpawnDaFire - can explore manual fire executions to see if we can cut costs and increase intuitive feel
  (if (and (Curie|IsCurieActive) (Math|Conversions|ToBoolean(Integer) (Math|Float|Truncate (ScalableFloat|GetValueatLevel (Variables|Default|GetRowFireEnabled)))))
    (Building|SpawnBuildingGameplayActor (Variables|Default|GetFireMesh) (Math|Transform|MakeTransform (Transformation|GetWorldLocation (Variables|Collision|GetCapsuleComponent))) (Variables|Actor|GetInstigator)) ; Fire's behaving much more intuitively with two.  Keeping it like this unless we need a perf win.
    (else
      _))
  (|CallParentFunction HitActors HitResults)
  ;[1 1 1] Play explosion sound and display sound indicator
  (Audio|PlaySoundatLocation _explodesound _returnvalue)
  (SoundIndicator|ShowSoundIndicator (Variables|Default|GetFortSoundIndicator) _returnvalue _explodesound 1.0 0))

(fn FindMeshANgle
  (params (HitLoc :type Vector) (HitNormal :type Vector))
  (return 0))

(fn OnRep_HideAndKill
  ;[1 1 1] If we set hide and kill to true disable collision set the actor to hidden in game and fade out.
  (if (Variables|Default|GetHideAndKill)
    (Rendering|SetActorHiddenInGame true)
    (Collision|SetActorEnableCollision false)
    (Audio|Components|Audio|FadeOut (Variables|Default|GetAudio) 0.2 0.0 "Linear" :!class /Script/Engine.AudioComponent)))

(fn ConstructionScript
  ;[1 1 1] Turn off the tick and set Generate Overlaps to false on the client.
  (if (not (Networking|IsServer))
    (Variables|/Script/Engine.PrimitiveComponent|SetbGenerateOverlapEvents false (Variables|Collision|GetCapsuleComponent))))
