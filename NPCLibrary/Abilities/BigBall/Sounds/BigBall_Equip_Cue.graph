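# Sound cue for equipping the NPC BigBall ability, playing a lightning start sound.
# A Random node picks evenly between two lightning start waves and feeds a Modulator at the root.
# It plays in the melee weapon sound class at reduced volume with the BigBall ability attenuation.
# A concurrency override allows one instance at a time and blocks new plays while it sounds.
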SoundCue /NPCLibrary/Abilities/BigBall/Sounds/BigBall_Equip_Cue
root=Modulator_0
nodes [
 {id=Modulator_0, children [Random_0]}
 {id=Random_0, children [WavePlayer_5,WavePlayer_0], properties {weights [1,1]}}
 {id=WavePlayer_0, properties {soundWaveAssetPtr=/NPCLibrary/Abilities/BigBall/Sounds/Waves/BigBall_Lightning_Start_02}}
 {id=WavePlayer_5, properties {soundWaveAssetPtr=/NPCLibrary/Abilities/BigBall/Sounds/Waves/BigBall_Lightning_Start_01}}
]
properties {
 attenuationSettings=/NPCLibrary/Abilities/BigBall/Sounds/Settings/BigBall_Ability_Att
 bOverrideConcurrency=true
 maxDistance=5500
 soundClassObject=&Fort_WeaponMelee
 volumeMultiplier=0.4
 concurrencyOverrides {maxCount=1, resolutionRule=PreventNew}
}
