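# Looping Sound Cue for the BigBall ability in the NPC library, classed as ranged weapon audio.
# A looping wave player feeds an Enveloper whose volume curve ramps from near silence to roughly full level over about a second and a half, rising slowly at first.
# The cue halves overall volume, uses the ability's attenuation asset and overrides concurrency to cap simultaneous instances at two.
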SoundCue /NPCLibrary/Abilities/BigBall/Sounds/BigBall_Loop_Cue
root=Enveloper_1
nodes [
 {
  id=Enveloper_1
  children [WavePlayer_5]
  properties {
   pitchCurve {editorCurveData {keys [{time=0}]}}
   volumeCurve {editorCurveData {keys [<at 0 0.001>,<at 0.767 0.0671>,<at 1.087 0.413>,<at 1.517 1.0333>]}}
  }
 }
 {id=WavePlayer_5, properties {bLooping=true, soundWaveAssetPtr=/NPCLibrary/Abilities/BigBall/Sounds/Waves/BigBall_Loop_v5}}
]
properties {
 attenuationSettings=/NPCLibrary/Abilities/BigBall/Sounds/Settings/BigBall_Ability_Att
 bDebug=true
 bOverrideConcurrency=true
 maxDistance=5500
 soundClassObject=&Fort_WeaponRanged
 volumeMultiplier=0.5
 concurrencyOverrides {maxCount=2}
 concurrencySet [null]
}
