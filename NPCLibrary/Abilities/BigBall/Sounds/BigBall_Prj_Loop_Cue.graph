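# BigBall_Prj_Loop_Cue is a SoundCue that plays a looping aura sound, named for the projectile of the NPC BigBall ability.
# A looping wave player for BigBall_Aura_Loop_v2 feeds a Modulator root that adds a slight random pitch variation.
# It uses the ranged weapon sound class and the close weapon projectile attenuation asset, alongside stored overrides for logarithmic distance falloff.
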SoundCue /NPCLibrary/Abilities/BigBall/Sounds/BigBall_Prj_Loop_Cue
root=Modulator_0
nodes [
 {id=Modulator_0, children [WavePlayer_1], properties {pitchMax=1.01, pitchMin=0.99}}
 {id=WavePlayer_1, properties {bLooping=true, soundWaveAssetPtr=/NPCLibrary/Abilities/BigBall/Sounds/Waves/BigBall_Aura_Loop_v2}}
]
properties {
 attenuationSettings=&Fort_Weapon_Projectile_Close_Athena_Att
 maxDistance=4950
 priority=0.5
 soundClassObject=&Fort_WeaponRanged
 volumeMultiplier=1
 attenuationOverrides {distanceAlgorithm=Logarithmic, falloffDistance=2000, stereoSpread=512, attenuationShapeExtents {x=512}}
}
