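# Sound cue for the BigBall ability's projectile explosion, layered by listener distance.
# A mixer blends close, medium and distant layers, each with its own attenuation asset, with the medium and distant layers mixed lower.
# Each layer randomly picks one of three waves and varies pitch; the close layer also varies volume and starts after a short fixed delay.
# A concurrency override caps simultaneous instances per owner.
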SoundCue /NPCLibrary/Abilities/BigBall/Sounds/BigBall_Proj_Expl_Cue
root=Mixer_0
nodes [
 {
  id=Attenuation_0
  children [Modulator_2]
  properties {attenuationSettings=/NPCLibrary/Abilities/BigBall/Sounds/Settings/TameFoot_BigBall_Medium_Explosion_Att}
 }
 {
  id=Attenuation_1
  children [Modulator_0]
  properties {attenuationSettings=/NPCLibrary/Abilities/BigBall/Sounds/Settings/TameFoot_BigBall_Distant_Explosion_Att}
 }
 {
  id=Attenuation_7
  children [Modulator_6]
  properties {attenuationSettings=/NPCLibrary/Abilities/BigBall/Sounds/Settings/BigBall_Close_Explosion_Att, attenuationOverrides {falloffDistance=1600}}
 }
 {id=Delay_0, children [Random_1], properties {delayMax=0.066, delayMin=0.066}}
 {id=Mixer_0, children [Attenuation_7, Attenuation_0, Attenuation_1], properties {inputVolume [1,0.7,0.7]}}
 {id=Modulator_0, children [Random_4], properties {pitchMax=1.1, pitchMin=0.9}}
 {id=Modulator_2, children [Random_2], properties {pitchMax=1.1, pitchMin=0.9}}
 {id=Modulator_6, children [Delay_0], properties {pitchMax=1.3, pitchMin=1, volumeMin=0.5}}
 {id=Random_1, children [WavePlayer_4,WavePlayer_5,WavePlayer_6], properties {weights [1,1,1]}}
 {id=Random_2, children [WavePlayer_0,WavePlayer_1,WavePlayer_2], properties {weights [1,1,1]}}
 {id=Random_4, children [WavePlayer_3,WavePlayer_7,WavePlayer_8], properties {weights [1,1,1]}}
 {id=WavePlayer_0, properties {soundWaveAssetPtr=/NPCLibrary/Abilities/BigBall/Sounds/Waves/BigBall_Proj_Explo_01_Med}}
 {id=WavePlayer_1, properties {soundWaveAssetPtr=/NPCLibrary/Abilities/BigBall/Sounds/Waves/BigBall_Proj_Explo_02_Med}}
 {id=WavePlayer_2, properties {soundWaveAssetPtr=/NPCLibrary/Abilities/BigBall/Sounds/Waves/BigBall_Proj_Explo_03_Med}}
 {id=WavePlayer_3, properties {soundWaveAssetPtr=/NPCLibrary/Abilities/BigBall/Sounds/Waves/BigBall_Proj_Explo_01_Far}}
 {id=WavePlayer_4, properties {soundWaveAssetPtr=/NPCLibrary/Abilities/BigBall/Sounds/Waves/BigBall_Proj_Explo_01}}
 {id=WavePlayer_5, properties {soundWaveAssetPtr=/NPCLibrary/Abilities/BigBall/Sounds/Waves/BigBall_Proj_Explo_02}}
 {id=WavePlayer_6, properties {soundWaveAssetPtr=/NPCLibrary/Abilities/BigBall/Sounds/Waves/BigBall_Proj_Explo_03}}
 {id=WavePlayer_7, properties {soundWaveAssetPtr=/NPCLibrary/Abilities/BigBall/Sounds/Waves/BigBall_Proj_Explo_02_Far}}
 {id=WavePlayer_8, properties {soundWaveAssetPtr=/NPCLibrary/Abilities/BigBall/Sounds/Waves/BigBall_Proj_Explo_03_Far}}
]
properties {
 bOverrideConcurrency=true
 maxDistance=15000
 soundClassObject=&Fort_WeaponRanged_Fire_3P_PSM
 concurrencyOverrides {bLimitToOwner=true, maxCount=5}
 concurrencySet [null]
}
