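# Sound cue for the throw of the NPC BigBall ability.
# A Random node picks one of three equally weighted throw wave variations.
# It plays through close-range projectile attenuation in the melee weapon sound class.
# A concurrency override allows a single instance at a time and blocks new plays while one is active.
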SoundCue /NPCLibrary/Abilities/BigBall/Sounds/BigBall_Throw
root=Random_0
nodes [
 {id=Random_0, children [WavePlayer_1,WavePlayer_3,WavePlayer_4], properties {weights [1,1,1]}}
 {id=WavePlayer_1, properties {soundWaveAssetPtr=/NPCLibrary/Abilities/BigBall/Sounds/Waves/BigBall_Throw_03}}
 {id=WavePlayer_3, properties {soundWaveAssetPtr=/NPCLibrary/Abilities/BigBall/Sounds/Waves/BigBall_Throw_02}}
 {id=WavePlayer_4, properties {soundWaveAssetPtr=/NPCLibrary/Abilities/BigBall/Sounds/Waves/BigBall_Throw_01}}
]
properties {
 attenuationSettings=&Fort_Weapon_Projectile_Close_Athena_Att
 bOverrideConcurrency=true
 maxDistance=4950
 soundClassObject=&Fort_WeaponMelee
 volumeMultiplier=1
 concurrencyOverrides {maxCount=1, resolutionRule=PreventNew}
}
