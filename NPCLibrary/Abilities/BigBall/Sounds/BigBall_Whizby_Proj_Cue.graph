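# BigBall_Whizby_Proj_Cue is a sound cue for the whiz-by of the BigBall NPC ability's projectile.
# A Random node picks between two equally weighted whiz-by waves, then a Modulator varies the pitch.
# The cue plays at reduced volume and slightly lowered pitch in a third-person ranged weapon fire sound class.
# Overridden attenuation disables volume falloff but keeps low-pass filtering, and a shared bullet whiz-by concurrency rule governs simultaneous plays.
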SoundCue /NPCLibrary/Abilities/BigBall/Sounds/BigBall_Whizby_Proj_Cue
root=Modulator_0
nodes [
 {id=Modulator_0, children [Random_1], properties {pitchMax=1.15}}
 {id=Random_1, children [WavePlayer_5,WavePlayer_7], properties {weights [1,1]}}
 {id=WavePlayer_5, properties {soundWaveAssetPtr=/NPCLibrary/Abilities/BigBall/Sounds/Waves/BigBall_WhizBy_02}}
 {id=WavePlayer_7, properties {soundWaveAssetPtr=/NPCLibrary/Abilities/BigBall/Sounds/Waves/BigBall_WhizBy_01}}
]
properties {
 bOverrideAttenuation=true
 maxDistance=2097152
 pitchMultiplier=0.95
 soundClassObject=&Fort_WeaponRanged_Fire_3P_PSM
 volumeMultiplier=0.25
 attenuationOverrides {bAttenuate=false, bAttenuateWithLPF=true, stereoSpread=2500}
 concurrencySet [/Game/Sounds/Concurrency_Rules/Athena_BulletWhizBy_Conc]
}
